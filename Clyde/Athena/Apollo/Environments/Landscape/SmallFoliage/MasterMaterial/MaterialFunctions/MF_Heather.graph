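# Material function that shades Apollo heather foliage, switchable via Use-MF_Heather against the incoming material attributes.
# Mask-textured green and flower colors, with optional Farm-mask variation, blend over landscape RVT color using grass biome masks and camera-distance fade.
# RVT normals optionally take heather normal detail, while UV1 shapes specular and height offset in gameplay grass-wind WPO.
# Without virtual texture support, a simpler color, roughness and opacity setup replaces the RVT blend.
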MaterialFunction /Clyde/Athena/Apollo/Environments/Landscape/SmallFoliage/MasterMaterial/MaterialFunctions/MF_Heather
graph {nodes [
  Constant() {r=-0.1}
  n1 = ScalarParameter() {defaultValue=1, parameterName=Roughness}
  n2 = TextureSampleParameter2D() {
   parameterName=HeatherMaskTexture
   samplerSource=SSM_Wrap_WorldGroupSettings
   samplerType=SAMPLERTYPE_Masks
   texture=/Clyde/Athena/Apollo/Environments/Landscape/SmallFoliage/Assets/Heather/Heather_M
  }
  n3 = Power(Base=@n2.R, Exp=ScalarParameter() {defaultValue=4, parameterName=Heather-OPpow})
  n4 = VectorParameter() {parameterName=Heather-ColorA, defaultValue=<rgba 0.0529 0.135 0.0203 1>}
  n5 = VectorParameter() {parameterName=HeatherGrasslands-ColorB, defaultValue=<rgba 0.4663 0.0825 0.55 1>}
  n6 = VectorParameter() {parameterName=Heather-ColorB, defaultValue=<rgba 0.4663 0.0825 0.55 1>}
  n7 = MaterialFunctionCall!(UseLandscapeCoords) {materialFunction=/Clyde/Athena/Apollo/Environments/Landscape/Materials/Biome/CommonMaterials/MF_ApolloTerrainMasks}
  n8 = VectorParameter() {parameterName=HeatherGrasslands-ColorB2, defaultValue=<rgba 0.4663 0.0825 0.55 1>}
  n9 = VectorParameter() {parameterName=Heather-ColorB2, defaultValue=<rgba 0.4663 0.0825 0.55 1>}
  n10 = StaticSwitchParameter(
   True=LinearInterpolate(A=LinearInterpolate(A=@n5.RGB, B=@n6.RGB, Alpha=@n7.Farm), B=LinearInterpolate(A=@n8.RGB, B=@n9.RGB, Alpha=@n7.Farm), Alpha=@n2.B)
   False=@n6.RGB
  ) {parameterName=UseMultiColor}
  n11 = Constant() {r=0.1}
  n12 = MaterialFunctionCall!(BiomeMask, UseBiomeColorMask, CliffMask, Grass-EdgeBiomeColorA, Grass-EdgeBiomeColorB, CombinedLandscapeLayersMask) {materialFunction=/Clyde/Athena/Apollo/Environments/Landscape/Materials/Biome/CommonMaterials/MF_Common_Grass}
  n13 = ComponentMask(None=TextureCoordinate() {coordinateIndex=1}) {g=true}
  n14 = LinearInterpolate(A=Constant() {r=1}, B=Constant() {r=-1}, Alpha=@n13) {constA=1.2, constB=-1}
  n15 = Multiply(A=@n12.GrassColorMaskA, B=Constant3Vector() {constant=<rgba 0 0 1 0>})
  n16 = MaterialFunctionCall!(
   SinkIntoGroundDistance
   WindFalloffMask=Saturate(None=Add(A=@n12.GrassColorMaskA, B=@n11))
   AfterHeightModification=Multiply(A=Saturate(None=@n14), B=Add(A=@n15, B=@n16.Wind))
  ) {materialFunction=/Game/Athena/Apollo/Environments/Landscape/SmallFoliage/MasterMaterial/MaterialFunctions/MF_GrassWind-GameplayWPO}
  n17 = SetMaterialAttributes!(
   MaterialAttributes
   BaseColor=LinearInterpolate(A=@n4.RGB, B=@n10, Alpha=@n2.G)
   Specular=Constant() {r=0.25}
   Roughness=@n1
   OpacityMask=@n3
   Normal
   WorldPositionOffset=@n16.FinalWPO
  ) {attributeSetTypes [
    #1
    #2
    #3
    #4
    #5
    #6
   ]}
  n18 = RuntimeVirtualTextureSampleParameter!(UVs, World Position, Mip Bias=Constant() {r=4}) {
   materialType=BaseColor_Normal_Specular_YCoCg
   mipValueMode=RVTMVM_MipBias
   parameterName=RVT
   virtualTexture=/Game/Athena/Apollo/RVT/RVT_Apollo_Landscape
  }
  n19 = ScalarParameter() {defaultValue=0.2, parameterName=Heather-SpecularHigh}
  n20 = SphereMask(
   A=CameraPositionWS()
   B=WorldPosition().XYZ
   Radius=ScalarParameter() {defaultValue=6000, parameterName=Heather-DistanceFade}
   Hardness=Constant()
  )
  n21 = Multiply(A=ScalarParameter() {defaultValue=-0.2, parameterName=Heather-SpecLow}, B=LinearInterpolate(B=@n20, Alpha=Constant() {r=0.8}) {constA=1})
  n22 = TextureSample() {
   samplerSource=SSM_Wrap_WorldGroupSettings
   samplerType=SAMPLERTYPE_Normal
   texture=/Clyde/Athena/Apollo/Environments/Landscape/SmallFoliage/Assets/Heather/Heather_N
  }
  n23 = MaterialFunctionCall(BaseNormal=@n18.Normal, AdditionalNormal=Transform(None=@n22.RGB)) {materialFunction=&BlendAngleCorrectedNormals}
  n24 = Multiply(A=ScalarParameter() {defaultValue=0.1, parameterName=Heather-NormalBlend}, B=@n20)
  n25 = ViewProperty() {property=MEVP_RuntimeVirtualTextureOutputLevel}
  n26 = Multiply(A=@n25.InvProperty, B=ScalarParameter() {defaultValue=1.5, parameterName=Heather-RVT_DistNear})
  n27 = LinearInterpolate(
   A=VectorParameter() {parameterName=Heather-UnderColor, defaultValue=<rgba 0.0529 0.135 0.0202 1>}.RGB
   B=LinearInterpolate(A=@n4.RGB, B=@n10, Alpha=Saturate(None=@n26))
   Alpha=@n2.G
  )
  n28 = Multiply(A=@n12.GrassColorMaskC, B=ScalarParameter() {defaultValue=1, parameterName=Heather-FlowerMaskMod})
  n29 = ScalarParameter() {parameterName=Heather-FlowerMaskAdd}
  n30 = Multiply(A=Saturate(None=Add(A=@n28, B=@n29)), B=@n20)
  n31 = Add(A=ScalarParameter() {defaultValue=1, parameterName=BlueChannelAdd}, B=@n2.A)
  n32 = SetMaterialAttributes(
   BaseColor=Multiply(A=LinearInterpolate(A=@n18.BaseColor, B=@n27, Alpha=@n30), B=Saturate(None=@n31))
   Specular=Saturate(None=LinearInterpolate(A=@n19, B=@n21, Alpha=@n13))
   Roughness=Multiply(A=@n18.Roughness, B=@n1)
   OpacityMask=@n3
   Normal=StaticSwitchParameter(True=LinearInterpolate(A=@n18.Normal, B=@n23.Result, Alpha=@n24), False=@n18.Normal) {defaultValue=true, parameterName=UseCustomNormalTexture}
   WorldPositionOffset=@n16.FinalWPO
  ) {attributeSetTypes [
    #1
    #2
    #3
    #4
    #5
    #6
   ]}
  FunctionOutput(None=StaticSwitchParameter(True=VirtualTextureFeatureSwitch!(No=@n17, Yes=@n32), False=FunctionInput() {inputType=FunctionInput_MaterialAttributes}) {group=ChooseMaterial, parameterName=Use-MF_Heather})
  Saturate(None=Divide(A=@n25.Property, B=ScalarParameter() {defaultValue=18, parameterName=Heather-RVT_MipDistanceFarEffect}))
 ]}
