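# Master material for instanced, Nanite-capable small landscape foliage in the Apollo environment: masked, two-sided, subsurface, with non-tangent-space normals.
# Material attributes chain through the SwampGrass, Heather and BasicKelpLinenCliff functions, then Subsurface Color is overridden with a zero constant.
# Opacity blends between SSSOpacity Bottom and Top parameters using a camera-to-instance-origin sphere mask, despite their height-based names.
# A separate UV-gradient SSSOpacity reroute is declared but unused.
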Material /Clyde/Athena/Apollo/Environments/Landscape/SmallFoliage/MasterMaterial/M_LS_GrassMaster_JuiceCliff
graph {
 properties {
  bTangentSpaceNormal=false
  bUseMaterialAttributes=true
  bUsedWithInstancedStaticMeshes=true
  bUsedWithNanite=true
  blendMode=BLEND_Masked
  pixelDepthOffsetMode=PDOM_Legacy
  shadingModel=MSM_Subsurface
  twoSided=true
  usedShadingModels=Subsurface
 }
 nodes [
  n1 = ScalarParameter() {defaultValue=1, group=SubSurface Settings, parameterName=SSSOpacity}
  n2 = ScalarParameter() {defaultValue=0.7, group=SubSurface Settings, parameterName=SSSOpacity2}
  n3 = Subtract(
   A=MaterialFunctionCall() {materialFunction=&LinearGradient}.VGradient
   B=ScalarParameter() {defaultValue=0.89, group=SubSurface Settings, parameterName=SSS gradient adjust}
  )
  NamedRerouteDeclaration(None=Saturate(None=LinearInterpolate(A=@n1, B=@n2, Alpha=@n3))) {name=SSSOpacity, nodeColor=<rgba 0.1719 0 1 1>}
  n4 = MaterialFunctionCall!(In=SetMaterialAttributes!(MaterialAttributes)) {materialFunction=/Clyde/Athena/Apollo/Environments/Landscape/SmallFoliage/MasterMaterial/MaterialFunctions/MF_SwampGrass}
  n5 = ScalarParameter() {defaultValue=1, group=GrassTypeSettings, parameterName=SSSOpacity_Bottom}
  n6 = ScalarParameter() {defaultValue=0.5, group=GrassTypeSettings, parameterName=SSSOpacity_Top}
  n7 = SphereMask(
   A=CameraPositionWS()
   B=TransformPosition(None=Constant3Vector()) {transformSourceType=TRANSFORMPOSSOURCE_Instance, transformType=TRANSFORMPOSSOURCE_World}
   Radius=ScalarParameter() {defaultValue=6000, group=GrassTypeSettings, parameterName=GrassV2_PerInstanceNormal_Distance}
   Hardness=Constant()
  )
  SetMaterialAttributes!(
   MaterialAttributes=MaterialFunctionCall!(In=MaterialFunctionCall!(In=@n4.Result) {materialFunction=/Clyde/Athena/Apollo/Environments/Landscape/SmallFoliage/MasterMaterial/MaterialFunctions/MF_Heather}.Result) {materialFunction=/Clyde/Athena/Apollo/Environments/Landscape/SmallFoliage/MasterMaterial/MaterialFunctions/MF_BasicKelpLinenCliff}.Result
   Subsurface Color=Constant()
   Opacity=Saturate(None=LinearInterpolate(A=@n5, B=@n6, Alpha=@n7))
  ) {attributeSetTypes [#1, #2]}
 ]
}
