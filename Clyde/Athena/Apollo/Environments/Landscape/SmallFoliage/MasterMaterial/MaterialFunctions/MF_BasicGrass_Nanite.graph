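# Builds material attributes for Apollo small landscape grass (Nanite variant), selected by the Use-MF_BasicGrass static switch.
# Base attributes come from landscape RVT or the mobile common-grass function, then gain dry tinting, tall-blade brightening, optional flower coloration and distance-masked specular.
# Landscape grass masks and vertex-color blue drive height offset, also output separately, feeding MF_GrassWind-GameplayWPO for world position offset.
# Forward and mobile paths simplify shading; many experimental branches remain unconnected.
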MaterialFunction /Clyde/Athena/Apollo/Environments/Landscape/SmallFoliage/MasterMaterial/MaterialFunctions/MF_BasicGrass_Nanite
graph {nodes [
  n1 = ComponentMask(None=TextureCoordinate() {coordinateIndex=1}) {g=true}
  n2 = Saturate(None=LinearInterpolate(A=Constant() {r=1}, B=Constant() {r=-3}, Alpha=@n1) {constA=1.2, constB=-1})
  n3 = VertexColor()
  n4 = OneMinus(None=@n1)
  n5 = MaterialFunctionCall!(CombinedLandscapeLayersMask, BypassUseRVT, BypassBiomeColors) {materialFunction=/Game/Athena/Apollo/Environments/Landscape/Materials/Biome/CommonMaterials/MF_Common_Grass}
  n6 = Constant() {r=2}
  n7 = Saturate(None=Add(A=Multiply(A=@n5.GrassColorMask2, B=@n6), B=Constant() {r=-1.25}))
  n8 = Constant() {r=0.6}
  n9 = Saturate(None=Add(A=Multiply(A=@n7, B=@n8), B=@n5.GrassColorMask1))
  n10 = LinearInterpolate(
   A=ScalarParameter() {parameterName=GrassHeightOffset_Min}
   B=ScalarParameter() {defaultValue=70, parameterName=GrassHeightOffset_Max}
   Alpha=@n9
  )
  n11 = Constant() {r=1}
  n12 = Multiply(
   A=@n3.B
   B=Multiply(A=ShadingPathSwitch(Default=@n10, Deferred=@n10, Forward=@n11, Mobile=@n11), B=Constant3Vector() {constant=<rgba 0 0 1 0>})
  )
  n13 = MaterialFunctionCall!(
   SinkIntoGroundDistance
   WindFalloffMask=@n9
   AfterHeightModification=Multiply(A=LinearInterpolate(A=@n2, B=@n4, Alpha=@n3.B), B=Add(A=@n12, B=@n13.Wind))
  ) {materialFunction=/Game/Athena/Apollo/Environments/Landscape/SmallFoliage/MasterMaterial/MaterialFunctions/MF_GrassWind-GameplayWPO}
  n14 = MaterialFunctionCall(Float3=@n13.WindNormals) {materialFunction=&BreakOutFloat3Components}
  Abs(None=Subtract(A=Abs(None=@n14.R), B=Abs(None=@n14.G)))
  Add(A=Saturate(None=Subtract(A=@n14.R, B=@n14.R)), B=Saturate(None=Subtract(A=@n14.G, B=@n14.G)))
  n15 = Multiply(A=Time(), B=ScalarParameter() {defaultValue=0.3, parameterName=r4ythji0})
  n16 = Constant2Vector() {g=1}
  n17 = TextureCoordinate() {coordinateIndex=1}
  n18 = ScalarParameter() {defaultValue=1, parameterName=asdgt54yukb6722}
  n19 = Add(A=PerInstanceRandom(), B=Multiply(A=Multiply(A=@n17, B=@n18), B=ScalarParameter() {defaultValue=1, parameterName=lolwut}))
  n20 = Divide(
   A=ComponentMask(None=WorldPosition() {worldPositionShaderOffset=WPT_ExcludeAllShaderOffsets}.XYZ) {g=true, r=true}
   B=ScalarParameter() {defaultValue=200, group=WPO, parameterName=WindWaveScale33}
  )
  n21 = TextureSample(UVs=Add(A=Multiply(A=@n15, B=@n16), B=Add(A=@n19, B=@n20))) {samplerSource=SSM_Wrap_WorldGroupSettings, samplerType=SAMPLERTYPE_LinearColor, texture=/Clyde/ExtraDependencies/T_TorusFlowmap}
  n22 = ComponentMask(None=Multiply(A=Subtract(A=@n21.RGB) {constB=0.5}, B=VectorParameter() {parameterName=hgre89ooo, defaultValue=<rgba 0.1 0 0 1>}.RGB)) {g=true, r=true}
  n23 = ConstantBiasScale(None=@n14.G)
  n24 = ScalarParameter() {defaultValue=2, parameterName=yrtyurth321}
  Add(A=TextureCoordinate(), B=LinearInterpolate(A=Multiply(A=@n4, B=@n22), Alpha=Power(Base=@n23, Exp=@n24)) {constB=0})
  Constant3Vector() {constant=<rgba 1 0 0 0>}
  Constant3Vector() {constant=<rgba 0.1066 0.3177 0.146 1>}
  Constant3Vector()
  Constant() {r=6}
  Constant() {r=0.75}
  Constant()
  Constant() {r=1.5}
  Constant() {r=0.2}
  Constant() {r=5000}
  Constant() {r=0.2}
  Constant() {r=7500}
  Constant() {r=1500}
  Constant() {r=100}
  Constant() {r=-5}
  Constant() {r=100}
  Constant() {r=-5}
  n25 = ScalarParameter() {defaultValue=0.65, parameterName=Rough}
  n26 = Constant() {r=1}
  FeatureLevelSwitch(Default=@n25, ES2=@n26, ES3_1=@n26, SM4_REMOVED=@n25)
  n27 = Add(
   A=Multiply(A=@n4, B=ScalarParameter() {defaultValue=1.5, parameterName=grs-specuv1mask-mod})
   B=ScalarParameter() {defaultValue=-0.3, parameterName=grs-specuv1mask-add}
  )
  n28 = MaterialFunctionCall!(CombinedLandscapeLayersMask, BypassUseRVT, BypassBiomeColors) {materialFunction=/Game/Athena/Apollo/Environments/Landscape/Materials/Biome/CommonMaterials/Mobile/MF_Mobile_Common_Grass}
  n29 = SetMaterialAttributes!(
   MaterialAttributes=@n28.ForStaticMesh
   Normal=Transform(None=Constant3Vector() {constant=<rgba 0 0 1 0>}) {transformSourceType=TRANSFORMSOURCE_Instance}
  ) {attributeSetTypes [#1]}
  n30 = RuntimeVirtualTextureSampleParameter!(UVs, World Position=VertexInterpolator(VS=WorldPosition().XYZ).PS, Mip Level=Constant() {r=3}) {
   bEnableFeedback=false
   materialType=BaseColor_Normal_Specular_YCoCg
   mipValueMode=RVTMVM_MipLevel
   parameterName=RVT
   virtualTexture=/Game/Athena/Apollo/RVT/RVT_Apollo_Landscape
  }
  n31 = SetMaterialAttributes!(MaterialAttributes, BaseColor=@n30.BaseColor, Specular=@n30.BaseColor, Roughness=@n30.BaseColor, Normal=@n30.BaseColor) {attributeSetTypes [
    #2
    #3
    #4
    #1
   ]}
  n32 = VirtualTextureFeatureSwitch!(No=@n28.ForStaticMesh, Yes=@n31)
  n33 = GetMaterialAttributes(None=StaticSwitchParameter(True=@n29, False=StaticSwitchParameter(True=@n32, False=@n29) {defaultValue=true, parameterName=ActualUseRVT}) {desc="True= used in Creative mode as well", parameterName=IsStaticMeshFoliageSpawner}) {attributeGetTypes [
    #2
    #1
    #4
    #3
   ]}
  n34 = StaticSwitchParameter(True=@n33.Specular, False=@n5.Spec-CameraDistance) {parameterName=IsStaticMeshFoliageSpawner}
  n35 = SphereMask(
   A=CameraPositionWS()
   B=WorldPosition().XYZ
   Radius=ScalarParameter() {defaultValue=6500, parameterName=Terrain-GrsSpecHeightDist}
   Hardness=Constant()
  )
  n36 = LinearInterpolate(A=@n34, B=Multiply(A=Saturate(None=@n27), B=@n34), Alpha=@n35)
  n37 = Constant() {r=0.3}
  n38 = Constant() {r=0.9}
  n39 = VertexNormalWS()
  n40 = TextureSample() {samplerSource=SSM_Wrap_WorldGroupSettings, texture=/Game/Athena/Apollo/Environments/Landscape/SmallFoliage/Assets/BasicGrass/T_ApolloGrass01}
  n41 = Add(A=Multiply(B=PerInstanceRandom()) {constA=5}) {constB=-0.5}
  n42 = Multiply(
   A=LinearInterpolate(
    A=Constant3Vector() {constant=<rgba 0.3628 0.1375 0.6 1>}
    B=Constant3Vector() {constant=<rgba 0.4 0.3133 0.008 1>}
    Alpha=Saturate(None=@n41)
   )
   B=@n40.R
  )
  n43 = VertexColor()
  n44 = Saturate(None=LinearInterpolate(A=Constant() {r=-2}, B=Constant() {r=5}, Alpha=@n5.GrassColorMask4))
  n45 = Saturate(None=LinearInterpolate(A=Constant() {r=1}, B=Constant() {r=-0.5}, Alpha=VertexColor().B) {constA=1.2, constB=-1})
  n46 = LinearInterpolate(
   A=@n33.BaseColor
   B=LinearInterpolate(
    A=Constant3Vector() {constant=<rgba 0.15 0.1163 0.015 1>}
    B=Constant3Vector() {constant=<rgba 0.13 0.0877 0.039 1>}
    Alpha=PerInstanceRandom()
   )
   Alpha=Multiply(A=@n45, B=@n44)
  )
  n47 = VertexColor()
  n48 = LinearInterpolate(A=@n46, B=Multiply(A=@n46, B=ScalarParameter() {defaultValue=1.5, parameterName=BrightenTall}), Alpha=@n47.B)
  n49 = ShadingPathSwitch(Default=@n48, Deferred=@n48, Forward=@n33.BaseColor, Mobile=@n33.BaseColor)
  n50 = Add(A=@n13.FinalWPO, B=Constant3Vector() {constant=<rgba 0 0 -2.5 1>})
  n51 = SetMaterialAttributes!(
   MaterialAttributes
   BaseColor=StaticSwitchParameter(True=LinearInterpolate(A=@n49, B=@n42, Alpha=@n43.B), False=@n49) {parameterName=UseFlowerColoration}
   Specular=ShadingPathSwitch(Default=@n36, Deferred=@n36, Forward=@n37, Mobile=@n37)
   Roughness=ShadingPathSwitch(Default=@n33.Roughness, Deferred=@n33.Roughness, Forward=@n38, Mobile=@n38)
   OpacityMask=Multiply(A=@n40.A, B=Constant() {r=4}) {constB=4}
   Normal=ShadingPathSwitch(Default=@n33.Normal, Deferred=@n33.Normal, Forward=@n39, Mobile=@n39)
   WorldPositionOffset=@n50
  ) {attributeSetTypes [
    #2
    #3
    #4
    #5
    #1
    #6
   ]}
  FunctionOutput(None=StaticSwitchParameter(True=@n51, False=FunctionInput() {inputType=FunctionInput_MaterialAttributes}) {group=ChooseMaterial, parameterName=Use-MF_BasicGrass})
  FunctionOutput(None=@n12) {outputName=WpoHeight Modification, sortPriority=66}
  n52 = Multiply(
   A=ConstantBiasScale(None=ComponentMask(None=@n13.WindNormals) {g=true})
   B=ScalarParameter() {defaultValue=1.2, parameterName=2specwavemask-mod}
  )
  n53 = ScalarParameter() {defaultValue=-0.2, parameterName=2specwavemask-add}
  n54 = LinearInterpolate(
   A=ScalarParameter() {parameterName=GrsSpecLowerNear2}
   B=ScalarParameter() {defaultValue=-0.5, parameterName=GrsSpecLowerNear}
   Alpha=Saturate(None=Add(A=@n52, B=@n53))
  )
  LinearInterpolate(A=@n34, B=Saturate(None=LinearInterpolate(A=@n34, B=@n54, Alpha=@n1)), Alpha=@n35)
  n55 = Multiply(A=SphereMask(A=CameraPositionWS(), B=WorldPosition().XYZ, Radius=Constant() {r=2500}, Hardness=Constant()), B=PerInstanceRandom())
  LinearInterpolate(
   A=Multiply(A=@n33.BaseColor, B=Constant() {r=0.9})
   B=Multiply(A=Constant() {desc=Tip brightness, r=2}, B=@n33.BaseColor)
   Alpha=Multiply(A=Multiply(A=@n55, B=@n45), B=OneMinus(None=@n44))
  )
  n56 = ScalarParameter() {defaultValue=-0.25, parameterName=DF-OpacityAddFar}
  n57 = Multiply(A=ScalarParameter() {defaultValue=2, parameterName=DF-OpacityModFar}, B=@n40.A)
  n58 = ScalarParameter() {defaultValue=15, parameterName=DF-OpacityModNear}
  n59 = SphereMask(
   A=CameraPositionWS()
   B=WorldPosition().XYZ
   Radius=ScalarParameter() {defaultValue=1500, parameterName=DF-OpacityRadius}
   Hardness=Constant()
  )
  n60 = LinearInterpolate(A=Saturate(None=Add(A=@n56, B=@n57)), B=Saturate(None=Multiply(A=@n58, B=@n40.A)), Alpha=@n59)
  n61 = ScalarParameter() {defaultValue=0.6, parameterName=BasicGrass-OPpowFAR}
  n62 = ScalarParameter() {defaultValue=0.25, parameterName=BasicGrass-OPpowNEAR}
  n63 = ScalarParameter() {defaultValue=1, parameterName=DEBUG-grass-opacity}
  LinearInterpolate(A=@n60, B=Power(Base=@n40.A, Exp=LinearInterpolate(A=@n61, B=@n62, Alpha=@n59)), Alpha=@n63)
  LinearInterpolate(A=Constant3Vector() {constant=<rgba -1 -1 1 1>}, B=Constant3Vector() {constant=<rgba 0 0 1 1>}, Alpha=TwoSidedSign())
  LinearInterpolate(
   A=Multiply(A=ScalarParameter() {defaultValue=0.75, parameterName=DF-OpacityModFar01}, B=@n40.A)
   B=Multiply(A=@n40.A, B=ScalarParameter() {defaultValue=6, parameterName=DF-OpacityModNear01})
   Alpha=@n35
  )
  MaterialFunctionCall(
   1=Constant3Vector() {constant=<rgba 0.2243 0 1 1>}
   2=Constant3Vector() {constant=<rgba 0 1 0 0>}
   3=Constant3Vector() {constant=<rgba 0 0 1 0>}
   4=Constant3Vector() {constant=<rgba 1 0.5 0 1>}
   A=PerInstanceRandom()
  ) {materialFunction=&Lerp_Multiple_Float3}
  Multiply(A=@n63, B=@n60)
  OneMinus(None=@n3.B)
  Saturate(None=Multiply(A=@n47.B, B=@n4))
  n64 = Constant3Vector() {constant=<rgba 0 0 -3 1>}
  ShadingPathSwitch(Default=@n50, Deferred=@n50, Forward=@n64, Mobile=@n64)
  VertexInterpolator(VS=FunctionInput() {inputName=Pivots, sortPriority=10}) {desc=Removed because of flickerin bug}
  WorldPosition()
 ]}
