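# Material function for swamp grass foliage that builds its material attributes, switched by Use-MF_SwampGrass against a pass-through input.
# A two-color vertex-blue gradient, optionally multiplied by diffuse, gives base color, with opacity from diffuse alpha or a channel-selected mask.
# Normals either lean toward local up or blend with vertex normals.
# Biome grass masks drive wind falloff and a vertex-weighted Z offset through the shared gameplay grass-wind WPO function.
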MaterialFunction /Clyde/Athena/Apollo/Environments/Landscape/SmallFoliage/MasterMaterial/MaterialFunctions/MF_SwampGrass_Nanite
graph {nodes [
  ComponentMask(None=TextureCoordinate() {coordinateIndex=1}) {g=true}
  ComponentMask(None=TextureCoordinate() {coordinateIndex=1}) {g=true}
  Constant3Vector() {constant=<rgba 0 0 1 0>}
  Constant()
  Constant()
  n1 = TextureSampleParameter2D() {parameterName=Diffuse, texture=/Engine/EngineMaterials/DefaultDiffuse}
  n2 = VertexColor()
  n3 = LinearInterpolate(
   A=VectorParameter() {parameterName=Color, defaultValue=<rgba 0.1 0.0874 0.005 1>}.RGB
   B=VectorParameter() {parameterName=Color2, defaultValue=<rgba 0.0554 0.08 0.01 1>}.RGB
   Alpha=Multiply(A=ScalarParameter() {defaultValue=1, parameterName=ZGradientColor}, B=OneMinus(None=@n2.B))
  )
  n4 = TextureSampleParameter2D() {
   parameterName=OpacityMask
   samplerType=SAMPLERTYPE_Masks
   texture=/Game/Athena/Apollo/Environments/Landscape/SmallFoliage/Assets/SwampGrass/Swamp_Reeds_M
  }
  n5 = ChannelMaskParameter(None=@n4.RGBA) {parameterName=ChannelMask}
  n6 = Transform(None=Constant3Vector() {constant=<rgba 0 0 1 0>}) {transformSourceType=TRANSFORMSOURCE_Local}
  n7 = TextureSampleParameter2D() {
   parameterName=Normal
   samplerType=SAMPLERTYPE_Normal
   texture=/Game/Athena/Apollo/Environments/Landscape/SmallFoliage/Assets/SwampGrass/Swamp_Reeds_N
  }
  n8 = Transform(None=@n7.RGB)
  n9 = ScalarParameter() {defaultValue=1, parameterName=NormalIntensity}
  n10 = LinearInterpolate(A=@n6, B=@n8, Alpha=@n9)
  n11 = VertexNormalWS()
  n12 = MaterialFunctionCall(BaseNormal=@n11, AdditionalNormal=@n8) {materialFunction=&BlendAngleCorrectedNormals}
  n13 = Constant() {r=0.1}
  n14 = MaterialFunctionCall!(BiomeMask, UseBiomeColorMask, CliffMask, Grass-EdgeBiomeColorA, Grass-EdgeBiomeColorB, CombinedLandscapeLayersMask) {materialFunction=/Clyde/Athena/Apollo/Environments/Landscape/Materials/Biome/CommonMaterials/MF_Common_Grass}
  n15 = Constant() {r=2}
  n16 = Saturate(None=Add(A=Multiply(A=@n14.GrassColorMaskA2, B=@n15), B=Constant() {r=-1.25}))
  n17 = Constant() {r=0.6}
  n18 = Saturate(None=Add(A=Multiply(A=@n16, B=@n17), B=@n14.GrassColorMaskA))
  n19 = LinearInterpolate(
   A=ScalarParameter() {defaultValue=1, parameterName=Heather-wpoa}
   B=ScalarParameter() {defaultValue=-1, parameterName=Heather-wpob}
   Alpha=VertexColor().B
  ) {constA=1.2, constB=-1}
  n20 = Multiply(A=@n18, B=Constant3Vector() {constant=<rgba 0 0 1 0>})
  n21 = MaterialFunctionCall!(
   SinkIntoGroundDistance
   WindFalloffMask=Saturate(None=Add(A=@n18, B=@n13))
   AfterHeightModification=Multiply(A=Saturate(None=@n19), B=Add(A=@n20, B=@n21.Wind))
  ) {materialFunction=/Game/Athena/Apollo/Environments/Landscape/SmallFoliage/MasterMaterial/MaterialFunctions/MF_GrassWind-GameplayWPO}
  n22 = SetMaterialAttributes!(
   MaterialAttributes
   BaseColor=StaticSwitchParameter(True=Multiply(A=@n1.RGB, B=@n3), False=@n3) {parameterName=UseDiffuse}
   Specular=ScalarParameter() {defaultValue=0.05, parameterName=Specular}
   Roughness=ScalarParameter() {defaultValue=0.8, parameterName=Roughness}
   OpacityMask=Multiply(
    A=StaticSwitchParameter(True=@n1.A, False=@n5.R) {parameterName=UseDiffuse}
    B=ScalarParameter() {defaultValue=1.1, parameterName=OpacityMaskOffset}
   )
   Normal=StaticSwitchParameter(True=Normalize(VectorInput=@n10), False=LinearInterpolate(A=@n11, B=@n12.Result, Alpha=@n9)) {defaultValue=true, parameterName=UseForcedZup}
   WorldPositionOffset=@n21.FinalWPO
  ) {attributeSetTypes [
    #1
    #2
    #3
    #4
    #5
    #6
   ]}
  FunctionOutput(None=StaticSwitchParameter(True=@n22, False=FunctionInput() {inputType=FunctionInput_MaterialAttributes}) {group=ChooseMaterial, parameterName=Use-MF_SwampGrass})
  n23 = MaterialFunctionCall(Normal=@n7.RGB, Flatness=@n9) {materialFunction=&FlattenNormal}
  MaterialFunctionCall(BaseNormal=@n6, AdditionalNormal=@n23.Result) {materialFunction=&BlendAngleCorrectedNormals}
  Transform(None=@n23.Result)
 ]}
