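# Two-sided subsurface master material for Nanite landscape grass, layering swamp grass, heather and basic grass material functions with optional texture maps.
# UV1 pivots and vertex-color masks from SimpleGrassPivotPainter drive per-blade animation, blending wind, player push and physics-field velocity.
# Nanite gets rotation or length-preserving bend, while non-Nanite, ES3.1 and lower quality use a simpler offset.
# Switches add camera-relative math, fade-based sinking for fallback materials, height variation or disabled animation.
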Material /Clyde/Athena/Apollo/Environments/Landscape/SmallFoliage/MasterMaterial/M_LS_GrassMaster_Nanite
graph {
 properties {
  bTangentSpaceNormal=false
  bUseMaterialAttributes=true
  bUsedWithInstancedStaticMeshes=true
  bUsedWithNanite=true
  pixelDepthOffsetMode=PDOM_Legacy
  shadingModel=MSM_Subsurface
  twoSided=true
  usedShadingModels=Subsurface
 }
 nodes [
  n1 = TextureSampleParameter2D() {
   parameterName=Diffuse
   texture=/Game/Athena/Apollo/Environments/Landscape/SmallFoliage/Assets/FarmFlower/Textures/T_FarmFlowers_Grass_Nanite_basecolor
  }
  n2 = ScalarParameter() {defaultValue=1, parameterName=Brightness_Nanite}
  n3 = GetMaterialAttributes() {attributeGetTypes [#1]}
  n4 = Multiply(A=@n2, B=@n3.Base Color)
  n5 = Power(
   Base=Desaturation(None=@n1.RGB, Fraction=ScalarParameter() {defaultValue=1, parameterName=Desat_Nanite})
   Exp=ScalarParameter() {defaultValue=0.132, parameterName=ContrastNanite_Blend}
  )
  LinearInterpolate(
   A=Multiply(A=VectorParameter() {parameterName=Color_naniteBlend, defaultValue=<rgba 1 1 1 1>}.RGB, B=@n1.RGB)
   B=@n4
   Alpha=Saturate(None=@n5)
  )
  Multiply(A=@n5, B=@n2)
  n6 = TextureCoordinate() {coordinateIndex=1}
  n7 = MaterialFunctionCall(X=ComponentMask(None=@n6) {r=true}, Y=ComponentMask(None=@n6) {g=true}, Z=Constant()) {materialFunction=&MakeFloat3}
  NamedRerouteDeclaration(None=TransformPosition(None=@n7.Result) {transformSourceType=TRANSFORMPOSSOURCE_Instance, transformType=TRANSFORMPOSSOURCE_TranslatedWorld}) {name=PivotPosition_CameraRelative, nodeColor=<rgba 0 0.9766 1 1>}
  n8 = ScalarParameter() {defaultValue=1, group=SubSurface Settings, parameterName=SSSOpacity}
  n9 = ScalarParameter() {defaultValue=0.7, group=SubSurface Settings, parameterName=SSSOpacity2}
  n10 = Subtract(
   A=MaterialFunctionCall() {materialFunction=&LinearGradient}.VGradient
   B=ScalarParameter() {defaultValue=0.89, group=SubSurface Settings, parameterName=SSS gradient adjust}
  )
  NamedRerouteDeclaration(None=Saturate(None=LinearInterpolate(A=@n8, B=@n9, Alpha=@n10))) {name=SSSOpacity, nodeColor=<rgba 0.1719 0 1 1>}
  n11 = StaticSwitchParameter(True=NamedRerouteUsage(), False=WorldPosition().XYZ) {defaultValue=true, group=GrassTypeSettings_WPO, parameterName=UseSamplingPerBladeOrWorldPosition}
  n12 = Multiply(A=NamedRerouteUsage(), B=ScalarParameter() {defaultValue=200, group=GrassTypeSettings_WPO, parameterName=RandomSampleOffsetPerBlade})
  n13 = MaterialFunctionCall!(
   Position=StaticSwitchParameter(True=@n11, False=Add(A=@n12, B=@n11)) {group=GrassTypeSettings_WPO, parameterName="UseRealRotationAnimation(MostExpensive)"}
   IsTrees
   IsGrass=StaticBool() {value=true}
   IsMediumPlant
   IsMediumPlant_2024Anim
   IsMediumPlant2ndLevel
   MediumPlant_BasePivotPostion
  ) {materialFunction=/Game/Environments/Asteria/Foliage/Wind/MF_WindActorData}
  NamedRerouteDeclaration(None=@n13.WindIntensitySingleSample) {name=WindIntensityNoise, nodeColor=<rgba 0.1719 0 1 1>}
  n14 = TransformPosition(None=@n7.Result) {transformSourceType=TRANSFORMPOSSOURCE_Instance, transformType=TRANSFORMPOSSOURCE_World}
  NamedRerouteDeclaration(None=@n14) {name=PivotPosition, nodeColor=<rgba 0 0.9766 1 1>}
  n15 = ScalarParameter() {defaultValue=-50, group=GrassTypeSettings_WPO, parameterName=SinkIntoGroundAmount}
  n16 = Constant()
  n17 = PerInstanceFadeAmount()
  n18 = MaterialFunctionCall(Color=NamedRerouteUsage()) {materialFunction=&SplitComponents}
  n19 = CollectionParameter() {collection=&FortniteMaterialParameters, parameterName=GrassPushDownRadius}
  n20 = Divide(A=PixelDepth().R, B=@n19)
  n21 = Multiply(A=OneMinus(None=Saturate(None=@n20)), B=Constant() {r=-200})
  n22 = Constant()
  n23 = CollectionParameter() {collection=&FortniteMaterialParameters, parameterName=IsPlayerSliding}
  n24 = MaterialFunctionCall(X=@n18.R, Y=@n18.R, Z=Add(A=@n18.B, B=LinearInterpolate(A=@n21, B=@n22, Alpha=@n23))) {materialFunction=&MakeFloat3}
  n25 = Add(A=Multiply(A=LinearInterpolate(A=@n15, B=@n16, Alpha=@n17), B=Constant3Vector() {constant=<rgba 0 0 1 0>}), B=@n24.Result)
  NamedRerouteDeclaration(None=StaticSwitchParameter(
    True=Constant3Vector()
    False=StaticSwitchParameter(True=@n25, False=@n24.Result) {group=GrassTypeSettings, parameterName=IsFallbackMaterial}
   ) {group=GrassTypeSettings_WPO, parameterName=DisableAllAnimation}) {name=WPO-out, nodeColor=<rgba 0 1 0.3438 1>}
  n26 = VertexColor() {desc=Meshes using this material need the SimpleGrassPivotPainter.hda script run on them}
  NamedRerouteDeclaration(None=@n26.B) {name=B_VerticalGradient, desc="Vertical gradient - bottom=0 top=1", nodeColor=<rgba 1 0 0.0703 1>}
  NamedRerouteDeclaration(None=Frac(None=@n26.G)) {name=G_RandomPerBlade, desc=Random value per blade, nodeColor=<rgba 0.7109 0 1 1>}
  n27 = Normalize(VectorInput=CrossProduct(A=Constant3Vector() {constant=<rgba 0 0 1 0>}, B=NamedRerouteUsage()))
  n28 = Multiply(
   A=NamedRerouteUsage()
   B=LinearInterpolate(A=ScalarParameter() {defaultValue=0.5, group=GrassTypeSettings_WPO, parameterName=AnimBendFalloff}, Alpha=NamedRerouteUsage())
  )
  n29 = StaticSwitchParameter(True=NamedRerouteUsage(), False=NamedRerouteUsage()) {parameterName=UseCameraRelativeOptimization}
  n30 = StaticSwitchParameter(True=WorldPosition() {worldPositionShaderOffset=WPT_CameraRelativeNoOffsets}.XYZ, False=WorldPosition().XYZ) {parameterName=UseCameraRelativeOptimization}
  n31 = NamedRerouteUsage()
  n32 = LinearInterpolate(A=ScalarParameter() {defaultValue=0.5, group=GrassTypeSettings_WPO, parameterName=AnimBendFalloff_min}, Alpha=NamedRerouteUsage())
  n33 = ScalarParameter() {defaultValue=5, group=GrassTypeSettings_WPO, parameterName=AnimBendFalloff_Exponent}
  n34 = Multiply(
   A=Multiply(A=NamedRerouteUsage(), B=Power(Base=@n32, Exp=@n33))
   B=ScalarParameter() {defaultValue=150, group=GrassTypeSettings_WPO, parameterName=AnimStrength_Global}
  )
  n35 = StaticSwitchParameter(True=WorldPosition() {worldPositionShaderOffset=WPT_CameraRelativeNoOffsets}.XYZ, False=LocalPosition() {includedOffsets=ExcludeOffsets}.XYZ) {parameterName=UseCameraRelativeOptimization}
  n36 = StaticSwitchParameter(True=NamedRerouteUsage(), False=NamedRerouteUsage()) {parameterName=UseCameraRelativeOptimization}
  n37 = Subtract(A=Add(A=@n35, B=Multiply(A=@n31, B=@n34)), B=@n36)
  n38 = Length(None=Subtract(A=@n35, B=@n36)) {is=class MaterialExpressionLength}
  n39 = Length(None=@n37)
  n40 = Subtract(A=Multiply(A=@n37, B=Divide(A=@n38, B=@n39)), B=@n35)
  n41 = NamedRerouteUsage()
  n42 = Add(A=Constant() {r=1}, B=ScalarParameter() {defaultValue=1, group=GrassTypeSettings_WPO, parameterName=Stiffness})
  n43 = StaticSwitchParameter(True=@n41, False=Power(Base=@n41, Exp=Clamp(None=@n42) {maxDefault=50, minDefault=1}) {constExponent=3}) {group=GrassTypeSettings, parameterName=IsFlowers}
  n44 = LinearInterpolate(A=ScalarParameter() {defaultValue=-1, group=GrassTypeSettings_WPO, parameterName=LowPolyCounterOffset}, Alpha=@n43)
  n45 = NamedRerouteUsage()
  n46 = NamedRerouteUsage()
  n47 = Multiply(
   A=Multiply(
    A=StaticSwitchParameter(True=@n44, False=@n43) {group=GrassTypeSettings_WPO, parameterName=UseLowPolyCounterOffset}
    B=Multiply(A=@n45, B=@n46)
   )
   B=ScalarParameter() {defaultValue=150, group=GrassTypeSettings_WPO, parameterName=Master_Anim_strength}
  )
  n48 = QualitySwitch(
   Default=StaticSwitchParameter(True=RotateAboutAxis(NormalizedRotationAxis=@n27, RotationAngle=@n28, PivotPoint=@n29, Position=@n30), False=Add(A=@n40, B=@n36)) {group=GrassTypeSettings_WPO, parameterName="UseRealRotationAnimation(MostExpensive)"}
   Low=@n47
   Medium=@n47
  )
  NamedRerouteDeclaration(None=NaniteReplace!(Default=@n47, Nanite=FeatureLevelSwitch(Default=@n48, ES3_1=@n47))) {name=Simple Lookup Animation, nodeColor=<rgba 0 1 0.8359 1>}
  NamedRerouteDeclaration(None=@n13.WindI_Direction_WithNoise) {name=WindDirectionWithNoise, nodeColor=<rgba 0 1 0.9297 1>}
  n49 = CollectionParameter() {collection=&FortniteMaterialParameters, parameterName=PlayerPosition}
  n50 = ComponentMask(None=@n49) {b=true, g=true, r=true}
  n51 = StaticSwitchParameter(
   True=TransformPosition(None=@n50) {transformSourceType=TRANSFORMPOSSOURCE_World, transformType=TRANSFORMPOSSOURCE_TranslatedWorld}
   False=@n50
  ) {parameterName=UseCameraRelativeOptimization}
  n52 = StaticSwitchParameter(True=NamedRerouteUsage(), False=NamedRerouteUsage()) {parameterName=UseCameraRelativeOptimization}
  n53 = Constant() {r=200}
  n54 = Constant()
  NamedRerouteDeclaration(None=Saturate(None=SphereMask(A=@n51, B=@n52, Radius=@n53, Hardness=@n54) {attenuationRadius=64, hardnessPercent=20}) {desc=PlayerSphereMask}) {name=PlayerSphereMask, nodeColor=<rgba 0 0.8828 1 1>}
  NamedRerouteDeclaration(None=Normalize(VectorInput=Subtract(A=@n52, B=@n51))) {name=PlayerPushDirection, nodeColor=<rgba 0 0.6016 1 1>}
  n55 = SamplePhysicsVectorField!(World Position=NamedRerouteUsage()) {fieldTarget=Vector_LinearVelocity}
  n56 = MaterialFunctionCall(Vector 3=Add(A=@n55) {constB=0.0001}) {materialFunction=/Engine/Functions/Engine_MaterialFunctions02/Utility/VectorLength}
  n57 = Constant()
  NamedRerouteDeclaration(None=FeatureLevelSwitch(Default=Saturate(None=@n56.V3 Length), ES3_1=@n57)) {name=PhysicsFieldMask, nodeColor=<rgba 0 1 0.3906 1>}
  n58 = Constant3Vector() {constant=<rgba 0 0 1 0>}
  NamedRerouteDeclaration(None=FeatureLevelSwitch(
    Default=MaterialFunctionCall(Vector=@n55, Default=@n58) {materialFunction=&SafeNormalize}.Result
    ES3_1=Constant3Vector() {constant=<rgba 0 0 1 0>}
   )) {name=PhysicsFieldDirection, nodeColor=<rgba 0.3828 0 1 1>}
  n59 = Min(A=@n56.V3 Length, B=ScalarParameter() {defaultValue=1, group=GrassTypeSettings_WPO, parameterName=MaxRotation_PhysicsField_Clamp})
  n60 = Time()
  n61 = Multiply(A=ScalarParameter() {defaultValue=351, group=GrassTypeSettings_WPO, parameterName=PhysicsField_NoiseOffset}, B=NamedRerouteUsage())
  n62 = Frac(None=Multiply(
    A=ScalarParameter() {defaultValue=2.5, group=GrassTypeSettings_WPO, parameterName=MaxRotation_PhysicsField_NoiseSpeed}
    B=Add(A=@n60, B=@n61)
   ))
  n63 = OneMinus(None=@n62)
  n64 = LinearInterpolate(
   A=Multiply(A=Multiply(A=@n62, B=@n63)) {constB=2}
   Alpha=ScalarParameter() {defaultValue=0.6, group=GrassTypeSettings_WPO, parameterName=MaxRotation_PhysicsField_MInBend}
  )
  NamedRerouteDeclaration(None=FeatureLevelSwitch(Default=Multiply(A=@n59, B=@n64), ES3_1=@n57)) {name=PhysicsFieldStrength, nodeColor=<rgba 1 0 0.2109 1>}
  n65 = NamedRerouteUsage()
  n66 = NamedRerouteUsage()
  n67 = NamedRerouteUsage()
  NamedRerouteDeclaration(None=LinearInterpolate(A=LinearInterpolate(A=@n65, B=@n66, Alpha=@n67), B=NamedRerouteUsage(), Alpha=NamedRerouteUsage())) {name=BlendedDirectionVector, nodeColor=<rgba 0 1 0.3203 1>}
  n68 = Multiply(A=NamedRerouteUsage(), B=ScalarParameter() {defaultValue=0.38, group=GrassTypeSettings_WPO, parameterName=MaxRotation_texAnim})
  n69 = Multiply(A=NamedRerouteUsage(), B=ScalarParameter() {defaultValue=1, group=GrassTypeSettings_WPO, parameterName=MaxRotation_PhysicsField})
  n70 = NamedRerouteUsage()
  NamedRerouteDeclaration(None=LinearInterpolate(
    A=LinearInterpolate(A=@n68, B=@n69, Alpha=@n70)
    B=ScalarParameter() {defaultValue=0.375, group=GrassTypeSettings_WPO, parameterName=MaxRotation_PlayerPush}
    Alpha=NamedRerouteUsage()
   )) {name=AnimStrength, nodeColor=<rgba 1 0.3047 0 1>}
  NamedRerouteDeclaration(None=@n7.Result) {name=PivotPosition_LocalSpace, nodeColor=<rgba 0 0.9766 1 1>}
  n71 = SetMaterialAttributes!(MaterialAttributes)
  n72 = MaterialFunctionCall!(
   In=MaterialFunctionCall!(In=MaterialFunctionCall!(In=@n71) {materialFunction=/Clyde/Athena/Apollo/Environments/Landscape/SmallFoliage/MasterMaterial/MaterialFunctions/MF_SwampGrass_Nanite}.Result) {materialFunction=/Clyde/Athena/Apollo/Environments/Landscape/SmallFoliage/MasterMaterial/MaterialFunctions/MF_Heather_Nanite}.Result
   Pivots=NamedRerouteUsage()
  ) {materialFunction=/Clyde/Athena/Apollo/Environments/Landscape/SmallFoliage/MasterMaterial/MaterialFunctions/MF_BasicGrass_Nanite}
  n73 = Constant()
  n74 = CameraPositionWS()
  n75 = TransformPosition(None=Constant3Vector()) {transformSourceType=TRANSFORMPOSSOURCE_Instance, transformType=TRANSFORMPOSSOURCE_World}
  n76 = ScalarParameter() {defaultValue=6000, group=GrassTypeSettings, parameterName=GrassV2_PerInstanceNormal_Distance}
  n77 = Constant()
  n78 = Saturate(None=LinearInterpolate(
    A=ScalarParameter() {defaultValue=1, group=GrassTypeSettings, parameterName=SSSOpacity_Bottom}
    B=ScalarParameter() {defaultValue=0.5, group=GrassTypeSettings, parameterName=SSSOpacity_Top}
    Alpha=SphereMask(A=@n74, B=@n75, Radius=@n76, Hardness=@n77)
   ))
  n79 = NamedRerouteUsage()
  n80 = Add(A=@n72.WpoHeight Modification, B=@n79)
  n81 = GetMaterialAttributes(None=SetMaterialAttributes!(
    MaterialAttributes=SetMaterialAttributes!(MaterialAttributes=@n72.Result, Subsurface Color=@n73, Opacity=@n78) {attributeSetTypes [#2, #3]}
    World Position Offset=StaticSwitchParameter(True=@n80, False=@n79) {parameterName=UseGrassWPOHeightVar}
   ) {attributeSetTypes [#4]}) {attributeGetTypes [#5]}
  SetMaterialAttributes!(MaterialAttributes=@n81.MaterialAttributes, Specular=Multiply(A=@n81.Specular)) {attributeSetTypes [#5]}
  n82 = TextureSampleParameter2D() {parameterName=SMR, texture=/Game/Athena/Apollo/Environments/Landscape/SmallFoliage/Assets/FarmFlower/Textures/T_FarmFlowers_Grass_Nanite_SMR}
  n83 = TextureSampleParameter2D() {
   parameterName=Normal
   samplerType=SAMPLERTYPE_Normal
   texture=/Game/Athena/Apollo/Environments/Landscape/SmallFoliage/Assets/FarmFlower/Textures/T_FarmFlowers_Grass_Nanite_normal
  }
  n84 = SetMaterialAttributes!(
   MaterialAttributes=@n3.MaterialAttributes
   Base Color=@n4
   Specular=Multiply(A=@n82.R, B=ScalarParameter() {defaultValue=1, parameterName=Specular_nanite})
   Roughness=Multiply(A=@n82.B, B=ScalarParameter() {defaultValue=0.8, parameterName=Roughness_nanite})
   Normal=MaterialFunctionCall(Normal=@n83.RGB, Flatness=ScalarParameter() {parameterName=Normal_Flatten_nanite}) {materialFunction=&FlattenNormal}.Result
  ) {attributeSetTypes [
    #1
    #5
    #6
    #7
   ]}
  StaticSwitchParameter(True=@n84, False=@n3.MaterialAttributes) {defaultValue=true, parameterName=Use_Textures}
  VertexInterpolator(VS=@n14)
  VertexInterpolator(VS=Transform(None=Constant3Vector() {constant=<rgba 0 0 1 0>}) {transformSourceType=TRANSFORMSOURCE_Instance})
 ]
}
